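# Translucent unlit Niagara sprite and mesh particle material for an AstralFemale pickaxe effect, drawing a distorted radial wave shape.
# Polar-style UVs drive panning foam and cloud noise plus sine-warped noise, pushing UVs radially from the centre.
# Sphere masks shape opacity and a three-colour gradient, which particle colour overrides by default, with bottom masking and depth fade.
# Dynamic parameters offset UVs and animation time, shift the bottom mask, and control world-position stretch and pinch.
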Material /BRCosmetics/Weapons/FORT_Melee/Pickaxe_AstralFemale/M_AstralFemale_StaminaHawaii
graph {
 properties {
  bUsedWithNiagaraMeshParticles=true
  bUsedWithNiagaraSprites=true
  blendMode=BLEND_Translucent
  pixelDepthOffsetMode=PDOM_Legacy
  shadingModel=MSM_Unlit
  usedShadingModels=Unlit
 }
 nodes [
  n1 = TextureCoordinate()
  n2 = DynamicParameter() {defaultValue=<rgba 0 1 0 9>, paramNames [UV Offset Z, WPO Vertical Stretch, WPO Pinch Amount, Bottom Mask Offset]}
  n3 = AppendVector(A=Constant(), B=@n2.UV Offset Z)
  n4 = Subtract(A=NamedRerouteDeclaration(None=Subtract(A=@n1, B=@n3)) {name=Uv With Offset, nodeColor=<rgba 0 0.8594 1 1>}) {constB=0.5}
  n5 = SquareRoot(None=MaterialFunctionCall(Vector 2=@n4) {materialFunction=/Engine/Functions/Engine_MaterialFunctions02/Utility/VectorLength}.V2 Length)
  n6 = ScalarParameter() {defaultValue=1, group=Radial UV, parameterName=RadialMult, sliderMax=1}
  n7 = MaterialFunctionCall(Float2=@n4) {materialFunction=&BreakOutFloat2Components}
  n8 = Arctangent2Fast(Y=@n7.G, X=@n7.G)
  n9 = MaterialFunctionCall() {materialFunction=&Pi}
  n10 = Constant() {r=2}
  n11 = AppendVector(
   A=Power(Base=Multiply(A=@n5, B=@n6), Exp=ScalarParameter() {defaultValue=1, group=Radial UV, parameterName=RadialPower, sliderMax=5})
   B=Divide(
    A=Add(A=@n8, B=@n9.Result)
    B=MaterialFunctionCall(Multiplier=@n10) {materialFunction=&Pi}.Result
   )
  )
  n12 = Multiply(A=ComponentMask(None=@n11) {r=true}, B=ScalarParameter() {defaultValue=0.1, group=Distortion UV, parameterName=Distort Tile U})
  n13 = ComponentMask(None=@n11) {g=true}
  n14 = Multiply(A=@n13, B=ScalarParameter() {defaultValue=1.8, group=Distortion UV, parameterName=Distort Tile V})
  n15 = Time() {bOverride_Period=true, period=64}
  n16 = DynamicParameter() {parameterIndex=1, defaultValue=<rgba 0 1 1 1>, paramNames [Radial UV Offset, NA, NA, NA]}
  n17 = ScalarParameter() {defaultValue=-0.3, group=Distortion UV, parameterName=Distort Pan Speed U}
  n18 = ScalarParameter() {defaultValue=-0.175, group=Distortion UV, parameterName=Distort Pan Speed V}
  n19 = TextureSample(UVs=Panner(Coordinate=AppendVector(A=@n12, B=@n14), Time=Add(A=@n15, B=@n16.Radial UV Offset), Speed=AppendVector(A=@n17, B=@n18)) {speedX=-1, speedY=0.1}) {samplerType=SAMPLERTYPE_LinearColor, texture=/Game/Athena/Apollo/Water/Textures/02_Rounded/03_WaterFoam_Veroni_base_H}
  Add(A=@n19.R, B=ScalarParameter() {defaultValue=1, group=Distortion UV, parameterName=Distort L Bias})
  n20 = Constant()
  n21 = ComponentMask(None=TextureCoordinate()) {g=true}
  n22 = OneMinus(None=@n21)
  n23 = Add(A=@n22, B=ScalarParameter() {defaultValue=-0.5, group=WPO, parameterName=WPO Stretch Mask Offset})
  n24 = DynamicParameter() {defaultValue=<rgba 0 1 0 9>, paramNames [UV Offset Z, WPO Vertical Stretch, WPO Pinch Amount, Bottom Mask Offset]}
  n25 = ScalarParameter() {defaultValue=-0.5, group=WPO, parameterName=WPO Pinch Top Mask Offset}
  n26 = Constant()
  n27 = DynamicParameter() {defaultValue=<rgba 0 1 0 9>, paramNames [UV Offset Z, WPO Vertical Stretch, WPO Pinch Amount, Bottom Mask Offset]}
  n28 = ComponentMask(None=TextureCoordinate()) {r=true}
  n29 = Multiply(
   A=MaterialFunctionCall(R=@n26, G=@n27.WPO Pinch Amount, B=@n26) {materialFunction=&AppendMany}.RGB
   B=Add(A=Multiply(A=@n28) {constB=2}) {constB=-1}
  )
  n30 = Multiply(
   A=MaterialFunctionCall(In=Add(A=@n22, B=@n25), Contrast=ScalarParameter() {group=WPO, parameterName=WPO Pinch Top Contrast}) {materialFunction=&CheapContrast}.Result
   B=@n29
  )
  n31 = Add(A=@n21, B=ScalarParameter() {defaultValue=-0.8, group=WPO, parameterName=WPO Pinch Bottom Mask Offset})
  n32 = ScalarParameter() {defaultValue=1, group=WPO, parameterName=WPO Pinch Bottom Contrast}
  n33 = Multiply(
   A=Multiply(A=MaterialFunctionCall(In=@n31, Contrast=@n32) {materialFunction=&CheapContrast}.Result)
   B=@n29
  )
  Add(
   A=MaterialFunctionCall(R=@n20, G=@n20, B=Multiply(A=@n23, B=@n24.WPO Vertical Stretch)) {materialFunction=&AppendMany}.RGB
   B=Add(A=Transform(None=@n30) {transformSourceType=TRANSFORMSOURCE_Local}, B=Transform(None=@n33) {transformSourceType=TRANSFORMSOURCE_Local})
  )
  n34 = ParticleColor()
  n35 = OneMinus(None=ComponentMask(None=TextureCoordinate()) {g=true})
  n36 = DynamicParameter() {defaultValue=<rgba 0 1 0 9>, paramNames [UV Offset Z, WPO Vertical Stretch, WPO Pinch Amount, Bottom Mask Offset]}
  n37 = NamedRerouteUsage()
  n38 = Constant() {r=0.5}
  n39 = ScalarParameter() {defaultValue=0.4, group=Waves, parameterName=SphereMask Radius}
  n40 = ScalarParameter() {defaultValue=0.85, group=Waves, parameterName=SphereMask Hardness}
  n41 = ScalarParameter() {defaultValue=1, group=Waves, parameterName=SphereMask InsideRadius}
  n42 = ScalarParameter() {defaultValue=0.02, group=Waves, parameterName=SphereMask InsideHardness}
  n43 = Add(
   A=OneMinus(None=SphereMask(A=@n37, B=@n38, Radius=@n41, Hardness=@n42) {attenuationRadius=0.25, hardnessPercent=1})
   B=ScalarParameter() {group=Waves, parameterName=Sphere Mask Offset}
  )
  n44 = ScalarParameter() {defaultValue=0.4, group=Waves, parameterName=Sphere Mask Contrast}
  n45 = Multiply(
   A=MaterialFunctionCall(
    In=Add(A=@n35, B=@n36.Bottom Mask Offset)
    Contrast=ScalarParameter() {defaultValue=2.5, desc=Opacity, group=Opacity, parameterName=Bottom Mask Contrast}
   ) {materialFunction=&CheapContrast}.Result
   B=Multiply(
    A=SphereMask(A=@n37, B=@n38, Radius=@n39, Hardness=@n40) {attenuationRadius=0.25, hardnessPercent=1}
    B=MaterialFunctionCall(In=@n43, Contrast=@n44) {materialFunction=&CheapContrast}.Result
   )
  )
  n46 = OneMinus(None=SphereMask(
    A=NamedRerouteUsage()
    B=Constant() {r=0.5}
    Radius=ScalarParameter() {defaultValue=0.15, group=Opacity, parameterName=SphereMask SpikeCenter Radius}
    Hardness=ScalarParameter() {defaultValue=0.3, group=Opacity, parameterName=SphereMask SpikeCenter Hardness}
   ) {attenuationRadius=0.25, hardnessPercent=1})
  DepthFade(
   Opacity=Multiply(A=@n34.A, B=Multiply(A=@n45, B=@n46))
   FadeDistance=ScalarParameter() {defaultValue=100, group=Opacity, parameterName=Depth Fade}
  )
  n47 = VectorParameter() {parameterName=Color Dark, defaultValue=<rgba 0 0.4181 1.2 1>}
  n48 = VectorParameter() {parameterName=Color Mid, defaultValue=<rgba 0.5462 1.6856 2 0>}
  n49 = SphereMask(
   A=NamedRerouteUsage()
   B=Constant() {r=0.5}
   Radius=ScalarParameter() {defaultValue=0.35, group=Color, parameterName=Color SphereMask Radius}
   Hardness=ScalarParameter() {defaultValue=0.7, group=Color, parameterName=Color SphereMask Hardness}
  ) {attenuationRadius=0.25, hardnessPercent=1}
  n50 = NamedRerouteUsage()
  n51 = Constant() {r=0.5}
  n52 = ScalarParameter() {defaultValue=0.28, group=Color, parameterName=Color2 SphereMask Radius}
  n53 = ScalarParameter() {defaultValue=0.7, group=Color, parameterName=Color2 SphereMask Hardness}
  n54 = LinearInterpolate(
   A=LinearInterpolate(
    A=LinearInterpolate(A=@n47.RGB, B=@n48.RGB, Alpha=@n49)
    B=VectorParameter() {parameterName=Color Center, defaultValue=<rgba 3 3 3 0>}.RGB
    Alpha=SphereMask(A=@n50, B=@n51, Radius=@n52, Hardness=@n53) {attenuationRadius=0.25, hardnessPercent=1}
   )
   B=@n34.RGB
   Alpha=ScalarParameter() {defaultValue=1, group=Color, parameterName=Particle Color}
  )
  MaterialFunctionCall(Emissive=@n54) {materialFunction=&MF_FX_EmissiveAdjustments}
  n55 = Normalize(VectorInput=Subtract(A=TextureCoordinate(), B=Constant2Vector() {g=0.5, r=0.5}))
  n56 = Multiply(
   A=ScalarParameter() {defaultValue=1, parameterName=TimeMult}
   B=Add(A=Time(), B=DynamicParameter() {parameterIndex=1, paramNames [Radial UV Offset, NA, NA, NA]}.Radial UV Offset)
  )
  n57 = Sine(None=Add(A=@n13, B=Multiply(A=@n56) {constB=0.2}))
  n58 = TextureSample(UVs=AppendVector(A=Add(A=@n57, B=@n13), B=Constant() {r=0.5})) {samplerType=SAMPLERTYPE_Grayscale, texture=/FortWater/Textures/LowResBlurredNoise}
  n59 = MaterialFunctionCall(In=@n19.R, Contrast=ScalarParameter() {defaultValue=0.4, group=Distortion UV, parameterName=Distort L Contrast}) {materialFunction=&CheapContrast}
  n60 = AppendVector(
   A=Multiply(A=ComponentMask(None=@n11) {r=true}, B=ScalarParameter() {defaultValue=1, group=Distortion UV, parameterName=Distort2 Tile U})
   B=Multiply(A=ComponentMask(None=@n11) {g=true}, B=ScalarParameter() {defaultValue=6, group=Distortion UV, parameterName=Distort2 Tile V})
  )
  n61 = Add(
   A=Time() {bOverride_Period=true, period=64}
   B=DynamicParameter() {parameterIndex=1, defaultValue=<rgba 0 1 1 1>, paramNames [Radial UV Offset, NA, NA, NA]}.Radial UV Offset
  )
  n62 = AppendVector(
   A=ScalarParameter() {defaultValue=-1, group=Distortion UV, parameterName=Distort2 Pan Speed U}
   B=ScalarParameter() {defaultValue=-0.05, group=Distortion UV, parameterName=Distort2 Pan Speed V}
  )
  n63 = Multiply(
   A=TextureSample(UVs=Panner(Coordinate=@n60, Time=@n61, Speed=@n62) {speedX=-1, speedY=0.1}) {texture=&AtmosphericCloudNoise01}.R
   B=ScalarParameter() {defaultValue=0.8, group=Distortion UV, parameterName=Distort S Strength}
  )
  n64 = Multiply(
   A=Add(A=Multiply(A=@n58.R), B=Add(A=@n59.Result, B=@n63) {constB=0})
   B=ScalarParameter() {defaultValue=0.1, group=Distortion UV, parameterName=Distort Overal Strength}
  )
  NamedRerouteDeclaration(None=Add(A=TextureCoordinate(), B=Multiply(A=@n55, B=@n64))) {name=Distort UV, nodeColor=<rgba 0 1 0.625 1>}
  NamedRerouteDeclaration(None=@n54) {name=Color, nodeColor=<rgba 1 0.6797 0 1>}
  NamedRerouteUsage()
 ]
}
